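# Physics asset that provides the hitbox for the Vanquish vehicle body skeletal mesh.
# It has a single body on the root bone, made of one box shape offset forward and upward from the bone origin.
# There are no constraints, so the body is one rigid volume with no articulated parts.
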PhysicsAsset /VehicleCosmetics/Bodies/Vanquish/Meshes/Body_Vanquish_SK_HitBox
bodies {root [{shapeType=Box, extentX=386.3047, extentY=245.473, extentZ=133.309, shapeName=root_box, center=<72.8042 0 132.8319>}]}
constraints []
